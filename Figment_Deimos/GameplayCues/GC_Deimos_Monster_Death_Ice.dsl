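; Burst gameplay cue for the ice death effect of Deimos monsters, tailoring its Cascade particles to the enemy that died.
; On burst, it checks the target against Fiend and Brute gameplay tags and, on a match, zeroes the first particle component's relative location.
; It then feeds the target's enemy color into the MonsterColor parameter and the Brute check into BigEnemy as a float.
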(event EventBeginPlay)

(event Collision|EventActorBeginOverlap
  (params (OtherActor :type Actor%20Object%20Reference)))

(event EventTick
  (params (DeltaSeconds :type Float%20%28single-precision%29)))

(fn OnBurst(Cascade)
  (params (MyTarget :type Actor%20Object%20Reference) (Parameters :type Gameplay%20Cue%20Parameters%20Structure%20%28by%20ref%29) (ParticleComponents :type Array%20of%20Cascade%20Particle%20System%20Component%20Object%20References) (AudioComponents :type Array%20of%20Audio%20Component%20Object%20References) (BurstCameraShakeInstance :type Camera%20Shake%20Base%20Object%20Reference) (BurstDecalInstance :type Decal%20Actor%20Object%20Reference))
  (bind _output (Utilities|Array|Get(acopy) ParticleComponents))
  (bind _returnvalue (GameplayTags|HasAnyMatchingGameplayTags(Message) MyTarget (LocalVariables|GetFiendTags)))
  (bind _returnvalue_2 (GameplayTags|HasAnyMatchingGameplayTags(Message) MyTarget (LocalVariables|GetBruteTags)))
  (if (or _returnvalue _returnvalue_2)
    (Transformation|SetRelativeLocation _output "0, 0, 0" false false)
    (bind _color (Deimos|GetEnemyColor(Message) MyTarget))
    (Effects|Components|ParticleSystem|SetVectorParameter _output "MonsterColor" (Math|Conversions|ToVector(LinearColor) _color))
    (Effects|Components|ParticleSystem|SetFloatParameter _output "BigEnemy" (Math|Conversions|ToFloat(Boolean) _returnvalue_2))
    (else
      _))
  (return))

(fn ConstructionScript)
